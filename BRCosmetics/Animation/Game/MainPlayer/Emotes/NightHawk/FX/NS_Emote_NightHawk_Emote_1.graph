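# Niagara system for the NightHawk emote 1, a short ground-impact burst of dust and debris.
# Two flipbook smoke sprite emitters puff from cylinder rims, alongside gravity-pulled sprite bits and an occasional tumbling mesh chunk.
# The chunk generates location events, apparently feeding the curl-noise ribbon trail, which has no spawner of its own.
# Debris sizes switch between gameplay and front end; a system module sets a foot-dust color array index.
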NiagaraSystem /BRCosmetics/Animation/Game/MainPlayer/Emotes/NightHawk/FX/NS_Emote_NightHawk_Emote_1
dependencies {
 usedDataInterfaces [
  &NiagaraDataInterfaceCurve
  &NiagaraDataInterfaceColorCurve
  &NiagaraDataInterfaceCamera
  &NiagaraDataInterfaceVector2DCurve
 ]
 usedDynamicInputs [&Vector2DFromFloat]
 usedModules [
  &SystemState
  &self:SetFootDustFXColorArrayIndex
 ]
}
emitters {
 Directional_Smoke_Ground001 {
  emitterUpdateScript [
   &EmitterState()
   &SpawnBurst_Instantaneous(Spawn Count=6, Use Spawn Probability=false)
  ]
  particleSpawnScript [
   &InitializeParticle(
    Color=<rgba 1 1 1 0.1>
    Lifetime=&RandomRangeFloat(Minimum=0.22, Maximum=0.25)
    Sprite Rotation Mode=::NewEnumerator2
    Sprite Size Max=<15 90>
    Sprite Size Min=<12 65>
    Sprite Size Mode=::NewEnumerator2
    UsePositionOffset=false
    Write Material Random=true
   )
   &ShapeLocation(
    Cylinder Distribution=::NewEnumerator2
    Cylinder Height=0
    Cylinder Radius=15
    MaskSpawnGroup=false
    Rotation Coordinate Space=::World
    Shape Primitive=::NewEnumerator1
    U Position=&ReturnNormalizedExecIndex(Normalized Index Includes One=false)
    UseOwnerScale=true
   )
   &AddVelocity(
    Rotation Coordinate Space=::World
    Velocity=&Multiply_VectorByFloat(Vector=$Particles.ShapeLocation.ShapeNormal, Float=&RandomRangeFloat(Minimum=0.8, Maximum=1))
   )
   &AddVelocity(Rotation Coordinate Space=::World, Velocity=&RandomRangeVector(Minimum=<0 0 0.7>, Maximum=<0 0 0.8>)) {module=AddVelocity001}
   &ApplyInitialForces()
   /SetVariables(Particles.FirstFlipBookFrame:NiagaraInt32=&RandomRangeInt(Minimum=0, Maximum=64)) {isSetParameters=true, module=SetVariables_#1}
   /SetVariables(
    Particles.Stored Particle Size:Vector3f=$Particles.Scale
    Particles.Stored Sprite Size:Vector2f=&Multiply_Vector2D_ByFloat(Float=1, Vector2D=$Particles.SpriteSize)
   ) {isSetParameters=true, module=SetVariables_#2}
  ]
  particleUpdateScript [
   &ParticleState()
   /SetVariables(Particles.Scale=$StackContext.Stored Particle Size, Particles.SpriteSize=$StackContext.Stored Sprite Size) {isSetParameters=true, module=SetVariables_#3}
   &ScaleSpriteSize(Uniform Curve Sprite Scale=DataInterface() {propertyValues {Curve {keys [
        {arriveTangent=1.2275, interpMode=RCIM_Cubic, leaveTangent=1.2275, tangentMode=RCTM_User, time=0, value=0.2}
        {arriveTangent=1.7266, interpMode=RCIM_Cubic, leaveTangent=1.7266, tangentMode=RCTM_User, time=1, value=1.4}
       ]}}})
   &DynamicMaterialParameters(
    Index 0 Param 1=&FloatFromCurve(FloatCurve=DataInterface() {propertyValues {Curve {keys [<at 0 1 RCIM_Cubic>, {arriveTangent=-2.8182, interpMode=RCIM_Cubic, leaveTangent=-2.8182, tangentMode=RCTM_User, time=1, value=0}]}}})
    Index 0 Param 4=0.2
    Module.Param0WriteEnabled=true
    Module.Param1WriteEnabled=false
    Module.Param2WriteEnabled=false
    Module.Param3WriteEnabled=true
   ) {module=DynamicMatParams}
   &ScaleColor(
    Scale Alpha=&FloatFromCurve(FloatCurve=DataInterface() {propertyValues {Curve {keys [
         {arriveTangent=-0.2156, interpMode=RCIM_Cubic, leaveTangent=-0.2156, tangentMode=RCTM_User, time=0, value=1}
         {arriveTangent=-0.0478, interpMode=RCIM_Cubic, leaveTangent=-0.0478, tangentMode=RCTM_User, time=1, value=0}
        ]}}})
    Scale RGB=&MakeVectorFromLinearColorRGB(Color=&ColorFromCurve(ColorCurve=DataInterface() {propertyValues {
        AlphaCurve {keys [<at 0.3955 1 RCIM_Cubic>,<at 1 0 RCIM_Cubic>]}
        BlueCurve {keys [<at 0 0.2111 RCIM_Cubic>]}
        GreenCurve {keys [<at 0 0.3777 RCIM_Cubic>]}
        RedCurve {keys [<at 0 0.5104 RCIM_Cubic>]}
       }}))
    ScaleA=true
    ScaleRGB=true
   )
   &SubUVAnimation(
    End Frame:NiagaraInt32=&Add_Integer(A=$Particles.FirstFlipBookFrame, B=&Multiply_Int(A=&ReturnExecIndex(), B=8))
    Start Frame:NiagaraInt32=$Particles.FirstFlipBookFrame
    SubUV Loop Count=0
   )
   &SolveForcesAndVelocity(Clamp Velocity=false, Limit Acceleration=false)
  ]
  rendererClasses [&NiagaraSpriteRendererProperties]
  renderers [{rendererClass=&NiagaraSpriteRendererProperties}]
 }
 LargeDebris {
  emitterUpdateScript [
   &EmitterState()
   &SpawnBurst_Instantaneous(
    Spawn Count=&RandomRangeInt(Minimum=0, Maximum=1)
    Use Spawn Probability=false
    UseLoopCountLimit=false
   )
  ]
  particleSpawnScript [
   /SetVariables(Particles.NormalizedDebrisSize:NiagaraFloat=&RandomRangeFloat(Minimum=0, Maximum=1)) {isSetParameters=true}
   /SetVariables(Particles.MeshIndex=&RandomRangeInt(Minimum=0, Maximum=2)) {isSetParameters=true}
   &InitializeParticle(
    Color=<rgba 0.5677 0.4858 0.4572 1>
    Lifetime Max=1.5
    Lifetime Mode=::NewEnumerator1
    Mass Max=4
    Mass Min=0.5
    Mass Mode=::NewEnumerator3
    Mesh Scale Mode=::NewEnumerator3
    Mesh Uniform Scale=&FN_FrontEndFloatModifier(
     Gameplay=&Lerp_Float(A=0.06, B=0.12, Alpha=$Particles.NormalizedDebrisSize)
     Front End=&Lerp_Float(A=0.04, B=0.07, Alpha=$Particles.NormalizedDebrisSize)
    )
    Position Offset=<0 0 15>
    Sprite Rotation Mode=::NewEnumerator2
    UsePositionOffset=true
    Write Material Random=true
   )
   &ShapeLocation(MaskSpawnGroup=false, Override Seed=false, Sphere Radius=15, UseOwnerScale=true)
   &AddVelocity(Velocity=&Lerp_Vector(
     A=&RandomRangeVector(Minimum=<-200 -200 0>, Maximum=<200 200 1>)
     B=&RandomRangeVector(Minimum=<-150 -150 0>, Maximum=<150 150 1>)
     Alpha=$Particles.NormalizedDebrisSize
    ))
   &AddVelocity(
    Rotation Coordinate Space=::World
    Velocity=&Lerp_Vector(
     A=&RandomRangeVector(Minimum=<0 0 350>, Maximum=<1 1 650>)
     B=&RandomRangeVector(Minimum=<0 0 100>, Maximum=<1 1 150>)
     Alpha=0.5
    )
   ) {module=AddVelocity001}
   &MeshRotationForce(
    Coordinate Space:ENiagaraMeshTransforms=::NewEnumerator2
    Rotation=&RandomRangeVector(Minimum=<-2 -2 -2>, Maximum=<3 3 3>)
   )
   &ApplyInitialForces()
  ]
  particleUpdateScript [
   &ParticleState()
   &UpdateMeshOrientation(
    Rotation Rate=&Lerp_Float(A=2, B=1, Alpha=$Particles.NormalizedDebrisSize)
    Rotation Vector=&RandomVector(Vector Scale=1)
   )
   &GravityForce()
   &Drag(
    Drag=&Lerp_Float(A=1, B=0.5, Alpha=$Particles.NormalizedDebrisSize)
    Use Linear Drag=true
    Use Rotational Drag=false
   )
   &ScaleColor(
    Scale Alpha=&FloatFromCurve(FloatCurve=DataInterface() {propertyValues {Curve {keys [<at -0.0046 1.0113 RCIM_Cubic>,<at 1.0011 0.002 RCIM_Cubic>]}}})
    ScaleA=true
    ScaleRGB=false
   )
   &ScaleMeshSize(Scale Factor=&VectorFromFloat(Value=&FloatFromCurve(FloatCurve=DataInterface() {propertyValues {Curve {keys [
          {tangentMode=RCTM_User, time=0, value=1}
          {arriveTangent=-0.0405, interpMode=RCIM_Cubic, leaveTangent=-0.0405, tangentMode=RCTM_User, time=0.7009, value=1}
          {arriveTangent=-8.0089, interpMode=RCIM_Cubic, leaveTangent=-8.0089, tangentMode=RCTM_User, time=1, value=0}
         ]}}})))
   &SolveForcesAndVelocity(Clamp Velocity=false, Limit Acceleration=false)
   &GenerateLocationEvent(Event Probability=1, Event Send Rate=24, Use Event Probability=false, UseEventDelay=false)
  ]
  rendererClasses [&NiagaraMeshRendererProperties]
  renderers [{rendererClass=&NiagaraMeshRendererProperties, rendererIndex=0}]
 }
 Ribbon_Debris {
  emitterUpdateScript [&EmitterState()]
  particleSpawnScript [&InitializeRibbon(Color=<rgba 0.1979 0.1405 0.0948 0.1>, Lifetime=0.5, Ribbon Width=10, UsePositionOffset=false, Write Material Random=true)]
  particleUpdateScript [
   &ParticleState()
   &CurlNoiseForce(
    Noise Frequency=&FloatFromCurve(FloatCurve=DataInterface() {propertyValues {Curve {keys [<at 0.5 1>,<at 1 0.3>]}}}, Scale Curve=20)
    Noise Quality / Cost=::NewEnumerator0
    Noise Strength=&FloatFromCurve(
     FloatCurve=DataInterface() {propertyValues {Curve {keys [{tangentMode=RCTM_User, time=0.5, value=1}, {tangentMode=RCTM_User, time=1, value=2}]}}}
     Scale Curve=100
    )
    Pan Noise=false
   )
   &SolveForcesAndVelocity(Clamp Velocity=false, Limit Acceleration=false)
   &ScaleColor(
    Linear Color Curve=DataInterface() {propertyValues {
      AlphaCurve {keys [<at 0 1>,<at 1 0>]}
      BlueCurve {keys [<at 0 1>,<at 1 0.4583>]}
      GreenCurve {keys [<at 0 1>,<at 1 0.4583>]}
      RedCurve {keys [<at 0 1>,<at 1 0.4583>]}
     }}
    Scale Mode=::NewEnumerator2
   )
   &ScaleColor(
    Scale Alpha=&FloatFromCurve(
     FloatCurve=DataInterface() {propertyValues {Curve {keys [{interpMode=RCIM_Cubic, tangentMode=RCTM_User, time=1.1, value=1}, <at 2 0>]}}}
     CurveIndex=$Emitter.Age
    )
    ScaleA=true
    ScaleRGB=false
   ) {module=ScaleColor001}
   &ScaleRibbonWidth(Ribbon Width Scale=&FloatFromCurve(FloatCurve=DataInterface() {propertyValues {Curve {keys [{interpMode=RCIM_Cubic, tangentMode=RCTM_User, time=0.3, value=1}, {interpMode=RCIM_Cubic, tangentMode=RCTM_User, time=1, value=4}]}}}))
   &DynamicMaterialParameters(
    Index 0 Param 1=1
    Index 0 Param 2=20
    Index 0 Param 3=1
    Index 0 Param 4=1
    Module.Param0WriteEnabled=true
    Module.Param1WriteEnabled=true
    Module.Param2WriteEnabled=true
    Module.Param3WriteEnabled=true
   )
  ]
  rendererClasses [&NiagaraRibbonRendererProperties]
  renderers [{rendererClass=&NiagaraRibbonRendererProperties}]
 }
 SmallBits {
  emitterUpdateScript [
   &EmitterState()
   &SpawnBurst_Instantaneous(
    Spawn Count=&RandomRangeInt(Minimum=4, Maximum=8)
    Use Spawn Probability=false
    UseLoopCountLimit=false
   )
  ]
  particleSpawnScript [
   /SetVariables(Particles.randomInt:NiagaraInt32=&RandomRangeInt(Minimum=0, Maximum=1)) {isSetParameters=true}
   &InitializeParticle(
    Color Maximum=<rgba 0.4323 0.4287 0.3936 5>
    Color Minimum=<rgba 0.1088 0.1094 0.0992 5>
    Color Mode=::NewEnumerator2
    Lifetime=&RandomRangeFloat(Minimum=0.5, Maximum=3)
    Sprite Rotation Mode=::NewEnumerator2
    Sprite Size Mode=::NewEnumerator3
    Uniform Sprite Size=&FN_FrontEndFloatModifier(
     Gameplay=&RandomRangeFloat(Minimum=4, Maximum=6)
     Front End=&RandomRangeFloat(Minimum=2, Maximum=4)
    )
    UsePositionOffset=false
    Write Material Random=true
   )
   &SubUVAnimation(End Frame=3, SubUV Animation Mode:ENiagara_SubUVLookupMode=::NewEnumerator2)
   &ShapeLocation(
    Cylinder Height=5
    Cylinder Radius=20
    MaskSpawnGroup=false
    Offset=<0 0 25>
    Offset Mode=::NewEnumerator0
    Override Seed=false
    Shape Primitive=::NewEnumerator1
    Surface Only=false
    UseOwnerScale=true
   )
   /SetVariables(Particles.InitSpriteRotRate=&UniformRangedFloat(Minimum=-3, Maximum=3)) {isSetParameters=true}
   &AddVelocity(Velocity=&Multiply_VectorByFloat(Vector=$Particles.ShapeLocation.ShapeNormal, Float=&RandomRangeFloat(Minimum=200, Maximum=550)))
   &AddVelocity(Rotation Coordinate Space=::World, Velocity=&RandomRangeVector(Minimum=<0 0 50>, Maximum=<0 0 350>)) {module=AddVelocity001}
  ]
  particleUpdateScript [
   &ParticleState()
   &SpriteRotationRate(Rotation Rate=&Multiply_Float(
     A=$Particles.InitSpriteRotRate
     B=&FloatFromCurve(
      FloatCurve=DataInterface() {propertyValues {Curve {keys [
          {arriveTangent=-0.2085, interpMode=RCIM_Cubic, leaveTangent=-0.2085, tangentMode=RCTM_User, time=-0.0046, value=1.0113}
          {arriveTangent=-0.2496, interpMode=RCIM_Cubic, leaveTangent=-0.2496, tangentMode=RCTM_User, time=0.5, value=0.05}
         ]}}}
      CurveIndex=$Emitter.Age
      Scale Curve=950
     )
    ))
   &Collision(
    Age Colliding Particles=false
    Collision Enabled=&SetBoolByFloatComparison(A=$Emitter.Age, B=0.15)
    Control Roll On Collision=true
    Friction During a Bounce=0.2
    Module.Control Rest State Via Surface Normal=false
    Module.EnableMaxCollisionCount=false
    Module.Randomize Collision Normal Vector=true
    Restitution=0.5
    ScaleRateOfRotation=false
   ) {enabled=false}
   &GravityForce(Gravity=&RandomRangeVector(Minimum=<0 0 -980>, Maximum=<1 1 -1500>))
   &Drag(
    Drag=&FloatFromCurve(
     FloatCurve=DataInterface() {propertyValues {Curve {keys [{arriveTangent=-1, interpMode=RCIM_Cubic, leaveTangent=-3.4078, tangentMode=RCTM_Break, time=0, value=1}, <at 0.3 0.1>, <at 1 0>]}}}
     Scale Curve=7
    )
    Use Linear Drag=true
    Use Rotational Drag=false
   )
   &SolveForcesAndVelocity(Clamp Velocity=false, Limit Acceleration=false)
   &ScaleSpriteSize(
    Uniform Curve Scale=1.25
    Uniform Curve Sprite Scale=DataInterface() {propertyValues {Curve {keys [
        {arriveTangent=40, interpMode=RCIM_Cubic, time=0, value=0}
        {arriveTangent=0.1708, interpMode=RCIM_Cubic, leaveTangent=0.1708, tangentMode=RCTM_User, time=0.025, value=1}
        {arriveTangent=-3.2691, interpMode=RCIM_Cubic, leaveTangent=-3.2691, tangentMode=RCTM_User, time=1, value=0}
       ]}}}
   )
  ]
  rendererClasses [&NiagaraSpriteRendererProperties]
  renderers [{rendererClass=&NiagaraSpriteRendererProperties}]
 }
 Smoke_Ground {
  emitterUpdateScript [
   &EmitterState()
   &SpawnBurst_Instantaneous(Spawn Count=8, Use Spawn Probability=false)
  ]
  particleSpawnScript [
   &InitializeParticle(
    Color Maximum=<rgba 1 1 1 0.2>
    Color Minimum=<rgba 0.8594 0.8594 0.8594 0.2>
    Color Mode=::NewEnumerator2
    Lifetime=&RandomRangeFloat(Minimum=0.6, Maximum=0.7)
    Sprite Rotation Mode=::NewEnumerator2
    Sprite Size Mode=::NewEnumerator1
    Uniform Sprite Size Max=50
    Uniform Sprite Size Min=40
    UsePositionOffset=false
    Write Material Random=true
   )
   &ShapeLocation(
    Cylinder Distribution=::NewEnumerator2
    Cylinder Height=0
    Cylinder Radius=1
    MaskSpawnGroup=false
    Rotation Coordinate Space=::World
    Shape Primitive=::NewEnumerator1
    U Position=&ReturnNormalizedExecIndex(Normalized Index Includes One=false)
    UseOwnerScale=true
   )
   &AddVelocity(Velocity=&Multiply_VectorByFloat(Vector=$Particles.ShapeLocation.ShapeNormal, Float=&RandomRangeFloat(Minimum=250, Maximum=350)))
   &ApplyInitialForces()
   &CameraOffset(Camera Offset Amount=40)
   /SetVariables(Particles.FirstFlipBookFrame:NiagaraInt32=&RandomRangeInt(Minimum=0, Maximum=64)) {isSetParameters=true, module=SetVariables_#1}
   /SetVariables(
    Particles.Stored Particle Size:Vector3f=$Particles.Scale
    Particles.Stored Sprite Size:Vector2f=&Multiply_Vector2D_ByFloat(Float=1, Vector2D=$Particles.SpriteSize)
   ) {isSetParameters=true, module=SetVariables_#2}
  ]
  particleUpdateScript [
   &ParticleState()
   /SetVariables(Particles.Scale=$StackContext.Stored Particle Size, Particles.SpriteSize=$StackContext.Stored Sprite Size) {isSetParameters=true, module=SetVariables_#3}
   &AccelerationForce(Acceleration=&RandomRangeVector(Minimum=<0 0 -25>, Maximum=<1 1 -75>))
   &self:ViewpointBased_Sprite_Rotation_01(Rotation Rate=3) {module=ViewpointBased_Sprite_Rotation_01}
   &Drag(
    Drag=&Multiply_Float(
     A=&RandomRangeFloat(Minimum=10, Maximum=12)
     B=&FloatFromCurve(FloatCurve=DataInterface() {propertyValues {Curve {keys [
          {arriveTangent=-2.2549, interpMode=RCIM_Cubic, leaveTangent=-2.2549, tangentMode=RCTM_User, time=0, value=1}
          {interpMode=RCIM_Cubic, tangentMode=RCTM_User, time=1, value=0}
         ]}}})
    )
    Use Linear Drag=true
    Use Rotational Drag=true
   )
   &ScaleSpriteSize(Uniform Curve Sprite Scale=DataInterface() {propertyValues {Curve {keys [{arriveTangent=4.1286, interpMode=RCIM_Cubic, leaveTangent=4.1286, tangentMode=RCTM_User, time=0, value=0.2}, <at 1 1.4 RCIM_Cubic>]}}})
   &DynamicMaterialParameters(
    Index 0 Param 1=&FloatFromCurve(FloatCurve=DataInterface() {propertyValues {Curve {keys [<at 0 1 RCIM_Cubic>,<at 0.6 0 RCIM_Cubic>]}}})
    Index 0 Param 4=1
    Module.Param0WriteEnabled=true
    Module.Param1WriteEnabled=true
    Module.Param2WriteEnabled=false
    Module.Param3WriteEnabled=true
   ) {module=DynamicMatParams}
   &ScaleColor(
    Scale Alpha=&FloatFromCurve(FloatCurve=DataInterface() {propertyValues {Curve {keys [
         {arriveTangent=-0.2156, interpMode=RCIM_Cubic, leaveTangent=-0.2156, tangentMode=RCTM_User, time=0, value=1}
         {arriveTangent=0.1421, interpMode=RCIM_Cubic, leaveTangent=0.1421, tangentMode=RCTM_User, time=0.6, value=1}
         {arriveTangent=-0.9746, interpMode=RCIM_Cubic, leaveTangent=-0.9746, tangentMode=RCTM_User, time=1.0011, value=0}
        ]}}})
    Scale RGB=&MakeVectorFromLinearColorRGB(Color=&ColorFromCurve(ColorCurve=DataInterface() {propertyValues {
        AlphaCurve {keys [<at 0.3955 1 RCIM_Cubic>,<at 1 0 RCIM_Cubic>]}
        BlueCurve {keys [<at 0.0179 0.1262>,<at 0.9522 0.2448 RCIM_Cubic>]}
        GreenCurve {keys [<at 0.0179 0.2666>,<at 0.9522 0.517 RCIM_Cubic>]}
        RedCurve {keys [<at 0.0179 0.5156>,<at 0.9522 1 RCIM_Cubic>]}
       }}))
    ScaleA=true
    ScaleRGB=true
   )
   &SubUVAnimation(
    End Frame:NiagaraInt32=&Add_Integer(A=$Particles.FirstFlipBookFrame, B=64)
    Start Frame:NiagaraInt32=$Particles.FirstFlipBookFrame
    SubUV Loop Count:NiagaraFloat=&RandomRangeFloat(Minimum=0.25, Maximum=0.4)
   )
   &SolveForcesAndVelocity(Clamp Velocity=false, Limit Acceleration=false)
  ]
  rendererClasses [&NiagaraSpriteRendererProperties]
  renderers [{rendererClass=&NiagaraSpriteRendererProperties}]
 }
}
summary {
 systemName=NS_Emote_NightHawk_Emote_1
 emitters [
  {emitterName=LargeDebris, rendererClasses [&NiagaraMeshRendererProperties]}
  {emitterName=Ribbon_Debris, rendererClasses [&NiagaraRibbonRendererProperties]}
  {emitterName=Smoke_Ground, rendererClasses [&NiagaraSpriteRendererProperties]}
  {emitterName=Directional_Smoke_Ground001, rendererClasses [&NiagaraSpriteRendererProperties]}
  {emitterName=SmallBits, rendererClasses [&NiagaraSpriteRendererProperties]}
 ]
 userVariables [
  User.bIsFrontEnd {type=NiagaraBool}
  User.bIsFrontEndPreview {type=NiagaraBool}
  User.FootDustColorArray {
   type=NiagaraDataInterfaceArrayColor
   description=""
   defaultValue {
    struct=NiagaraExt_VariableValue_DataInterface
    value {
     dataInterface=&self:NiagaraDataInterfaceArrayColor_0
     dataInterfaceClass=&NiagaraDataInterfaceArrayColor
    }
   }
  }
  User.Global Scale {type=NiagaraFloat, defaultValue=0, description=""}
  User.SurfaceTypeIndex {type=NiagaraInt32}
 ]
}
